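; CoffeePuff's weapon Blueprint, which carries the gauntlet mesh and supports the hookshot ability.
; BeginPlay sets the gauntlet mesh on the server to avoid missing-mesh warnings.
; PlayLocalHookshotVO is cosmetic-only, checks the current voice line, alternates "Get Over Here!" and "Come Here!" via FlipFlop, then async-loads and attaches the sound.
; GA_CoffeePuff_Hookshot records the latest projectile here through an authority-only setter that forces a net update.
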(event EventBeginPlay
  ;[0.007147 0.15 0.02815 0.5] BeginPlay: Allow the server to find a valid weapon mesh to prevent warning logging.
  (|CallParentFunction)
  (Networking|SwitchHasAuthority
    (:Authority
      (Components|SkeletalMesh|SetSkeletalMeshAsset (Variables|Visual|GetWeaponMesh) "/CoffeePuffGameplay/Assets/Meshes/CoffeePuff_Gauntlet.CoffeePuff_Gauntlet"))))

(event Custom|PlayLocalHookshotVO
  ;[0.004483 0.15 0.1357 0.5] PlayLocalHookshotVO: This handles the playing of the iconic "Get Over Here!" and "Come Here!" audio.
  (Networking|CanExecuteCosmeticEvents
    (:True
      (bind _hookshotvo_audiocomponent (Variables|Default|GetHookshotVO_AudioComponent)
        (:then
          (if (Audio|Components|Audio|IsPlaying _hookshotvo_audiocomponent)
            _
            (else
              _)))
        (:else
          _))
      (bind _isa (Utilities|FlowControl|FlipFlop)
        (:A
          _)
        (:B
          _))
      (bind _object (Utilities|AsyncLoadAsset (select _isa (Variables|Sound|GetHookshotComeHereSoundCue) (Variables|Sound|GetHookshotGetOverHereSoundCue)))
        (:Completed
          (bind _assound_cue (Utilities|Casting|CastToSoundCue _object))
          (bind _returnvalue (Audio|SpawnSoundAttached :Sound _assound_cue :AttachToComponent (Variables|Visual|GetWeaponMesh) :AttachPointName "None" :LocationType "SnapToTarget" :bStopWhenAttachedToDestroyed false :VolumeMultiplier 1.0 :PitchMultiplier 1.0 :StartTime 0.0 :bAutoDestroy true))
          (Variables|Default|SetHookshotVO_AudioComponent _returnvalue))))))

(fn GetLatestHookshotProjectile
  ;[0.1087 0.02991 0.15 0.5] Returns the latest hookshot projectile.
  (return (Variables|Transient|GetLatestHookshotProjectile)))

(fn SetLatestHookshotProjectile
  (params (Projectile :type B%20Prj%20Coffee%20Puff%20Hookshot%20Object%20Reference))
  ;[0.0147 0.1414 0.15 0.5] SetLatestHookshotProjectile: Set the latest hookshot projectile fired by the weapon.
  ; Called from GA_CoffeePuff_Hookshot.
  (Networking|SwitchHasAuthority
    (:Authority
      (Variables|Transient|SetLatestHookshotProjectile Projectile)
      (Networking|ForceNetUpdate))))

(fn ConstructionScript
  (|CallParentFunction))
